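# Niagara skydiving contrail cosmetic "DangerInk", with ink trails on the character and a floating sign.
# Socket and skeletal mesh sampling anchors paired mesh trails, glows, dark hand smoke and golden drops.
# Sign, outline, aura, droplet and appear-smoke emitters follow System.SkullPosition, most bobbing less as glide lean rises.
# Emitter-age curves grow elements in, while lean, dive timeline, front-end preview and smoothed audio modulate motion, color and materials.
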NiagaraSystem /BRCosmetics/Athena/Effects/Particles/Skydiving/NS_Contrail_DangerInk/NS_Contrail_DangerInk
dependencies {
 usedDataInterfaces [
  &NiagaraDataInterfaceCurve
  &NiagaraDataInterfaceColorCurve
  &NiagaraDataInterfaceArrayFloat3
  &NiagaraDataInterfaceSocketReader
  &NiagaraDataInterfaceSkeletalMesh
 ]
 usedDynamicInputs [
  /Niagara/DynamicInputs/Integer/V2/Round_Float
  &TransformVector
  &SetBoolByFloatComparison
  &RotateVectorByQuaternion
  &Lerp_Vector
 ]
 usedModules [
  &self:SystemSpawnScript.SystemScriptSource.SystemScriptGraph.NiagaraNodeAssignment_4.SetVariables_#1
  &self:SystemSpawnScript.SystemScriptSource.SystemScriptGraph.NiagaraNodeAssignment_13.SetVariables_#2
  &SystemState
  &Timeline
  &self:SystemSpawnScript.SystemScriptSource.SystemScriptGraph.NiagaraNodeAssignment_1.SetVariables_#3
  &self:SystemSpawnScript.SystemScriptSource.SystemScriptGraph.NiagaraNodeAssignment_11.SetVariables_#4
 ]
}
emitters {
 AppearSmoke {
  emitterUpdateScript [
   &EmitterState(Life Cycle Mode=::NewEnumerator1, Loop Behavior=::NewEnumerator1, Loop Duration=0.4, UseLoopDelay=false)
   &SpawnBurst_Instantaneous(Spawn Count=7, Use Spawn Probability=false, UseLoopCountLimit=false)
  ]
  particleSpawnScript [&FN_Sprite_Attributes(Lifetime:NiagaraFloat=&RandomRangeFloat(Minimum=1.2, Maximum=1)) {module=FN_Sprite_v3}]
  particleUpdateScript [
   &ParticleState()
   &FN_Sprite_Attributes(
    Camera Offset=20
    Create Camera Offset Attribute=true
    Create Sprite Rotation Attribute=true
    Create SpriteSize Attribute=true
    SpriteRotation:NiagaraFloat=&RandomRangeFloat(Minimum=0, Maximum=360)
    SpriteSize=&Vector2DFromFloat(Value=&FloatFromCurve(
      FloatCurve=DataInterface() {propertyValues {Curve {keys [
          {
           arriveTangent=3.68
           arriveTangentWeight=2.0561
           interpMode=RCIM_Cubic
           leaveTangent=3.68
           leaveTangentWeight=2.0561
           tangentMode=RCTM_User
           time=0
           value=1
          }
          {tangentMode=RCTM_User, time=0.9949, value=3.019}
         ]}}}
      Scale Curve=&RandomRangeFloat(Minimum=20, Maximum=30)
     ))
   ) {module=FN_Sprite_v3001}
   /Game/Effects/Niagara/Modules/Sprite/FN_SpriteUV(Randomize Sprite UV scale=true, Seed=2, SubUV Lookup Index:NiagaraFloat=$Particles.NormalizedAge, Write to Module Namepsace=true)
   /SetVariables(Particles.BlackTurbulentColor:LinearColor=&ColorFromCurve(ColorCurve=DataInterface() {propertyValues {
       AlphaCurve {keys [<at 0 0 RCIM_Cubic>,<at 0.2454 0.9697>,<at 0.722 1>,<at 1 0>]}
       BlueCurve {keys [<at 0.0321 1 RCIM_Cubic>,<at 0.3175 0.03>,<at 0.8819 0.02 RCIM_Cubic>]}
       GreenCurve {keys [<at 0.0321 0.7317 RCIM_Cubic>,<at 0.3175 0.0106>,<at 0.8819 0.02 RCIM_Cubic>]}
       RedCurve {keys [<at 0.0321 0.747 RCIM_Cubic>,<at 0.3175 0.0024>,<at 0.8819 0.02 RCIM_Cubic>]}
      }})) {isSetParameters=true}
   /SetVariables(Particles.Position=&AddVectorToPosition(
     Position=Linked() {linkedVariable {name=System.SkullPosition, type=Vector3f}}
     Vector=/Niagara/DynamicInputs/Vectors/RotateVector(
      Vector To Rotate:Vector3f=&MakeVector(
       X=&FloatFromCurve(
        FloatCurve=DataInterface() {propertyValues {Curve {
           defaultValue=1
           keys [
            {arriveTangent=1.5741, interpMode=RCIM_Cubic, leaveTangent=1.5741, tangentMode=RCTM_User, time=0, value=0}
            {tangentMode=RCTM_User, time=1.5, value=1}
           ]
          }}}
        CurveIndex=$Emitter.Age
        Scale Curve=40
       )
       Y=&FloatFromCurve(
        FloatCurve=DataInterface() {propertyValues {Curve {
           defaultValue=1
           keys [
            {arriveTangent=1.5741, interpMode=RCIM_Cubic, leaveTangent=1.5741, tangentMode=RCTM_User, time=0, value=0}
            {tangentMode=RCTM_User, time=1.5, value=1}
           ]
          }}}
        CurveIndex=$Emitter.Age
        Scale Curve=40
       )
       Z=0
      )
      Yaw:NiagaraFloat=&RandomRangeFloat(Minimum=0, Maximum=360)
      Pitch:NiagaraFloat=0
      Roll:NiagaraFloat=0
      Rotation Rate:NiagaraFloat=1
     )
    )) {isSetParameters=true}
  ]
  rendererClasses [&NiagaraSpriteRendererProperties]
  renderers [{rendererClass=&NiagaraSpriteRendererProperties}]
 }
 Decal_Glow {
  emitterUpdateScript [
   &EmitterState(Life Cycle Mode=::NewEnumerator1, Loop Behavior=::NewEnumerator1, Loop Duration Mode=::NewEnumerator1, UseLoopDelay=false)
   &SpawnBurst_Instantaneous(Spawn Count=2, Use Spawn Probability=false, UseLoopCountLimit=false)
  ]
  particleSpawnScript [
   &InitializeParticle(
    Lifetime=&RandomRangeFloat(Minimum=1, Maximum=2)
    Sprite Size Mode=::NewEnumerator3
    Sprite UV Mode=::NewEnumerator3
    Uniform Sprite Size=45
    UsePositionOffset=false
    Write Material Random=true
   )
   &PartitionParticles(Write Partition to Ribbon ID=::NewEnumerator0)
   &AddVelocity(Velocity=<0 0 1>, Velocity Speed Scale=100)
   &ApplyInitialForces()
  ]
  particleUpdateScript [
   &ParticleState(
    Kill Particles When Lifetime Has Elapsed=false
    Let Infinitely Lived Particles Die When Emitter Deactivates=true
    Loop Particles Lifetime=true
   )
   &SocketLocation(
    SamplePosOffset=false
    Socket Index=$Particles.PartitionParticles.Partition
    Socket Reader=$System.Socket Reader
    Socket Sampling Mode=::NewEnumerator3
    VelocitySpeedLimit=false
   )
   &CameraOffset(Camera Offset Amount=10)
   &ScaleSpriteSize(Scale Factor=&Vector2DFromFloat(Value=&FloatFromCurve(
      FloatCurve=DataInterface() {propertyValues {Curve {keys [
          {
           interpMode=RCIM_Cubic
           leaveTangent=1.6793
           leaveTangentWeight=0.462
           tangentMode=RCTM_Break
           tangentWeightMode=RCTWM_WeightedBoth
           time=0
           value=0
          }
          {interpMode=RCIM_Cubic, tangentMode=RCTM_Break, time=1, value=1}
         ]}}}
      CurveIndex=$Emitter.Age
     )))
   &FN_Sprite_Attributes(
    Color=&MultiplyLinearColor(
     Linear Color=&MultiplyLinearColor(
      Linear Color=&MakeCustomLinearColorFromBool(Bool=$User.bIsFrontEndPreview, True Color=<rgba 0 0.1391 1 0.8>, False Color=<rgba 0.6458 0.7505 1 0.8>)
      Scale Factor=&MakeFloat4FromVectorAndFloat("Vector (RGB)"=<1 1 1>, "Float (Alpha)"=&FN_FrontEndFloatModifier(Gameplay=1, Front End=0.5))
     )
     Scale Factor=<1 1 1 1>
    )
    Create Color Attribute=true
    Create SpriteSize Attribute=true
    SpriteSize=&Multiply_Vector2D(A=$Particles.SpriteSize, B=<1.5 1.5>)
    Write to Module Namespace=true
   )
   &DynamicMaterialParameters(
    Index 0 Param 1=&Lerp_Float(A=1.5, B=0, Alpha=$System.SmoothedAudioScale)
    Index 0 Param 3=5
    Index 0 Param 4=&RandomRangeFloat(Minimum=0, Maximum=1)
    Index 2 Param 1=3
    Index 2 Param 2=1
    Index 2 Param 3=1
    Index 2 Param 4=1
    Module.Param0WriteEnabled=true
    Module.Param0WriteEnabled2=true
    Module.Param1WriteEnabled=true
    Module.Param1WriteEnabled2=true
    Module.Param2WriteEnabled=true
    Module.Param2WriteEnabled2=true
    Module.Param3WriteEnabled=true
    Module.Param3WriteEnabled2=true
    Write Parameter Index 2=true
   )
  ]
  rendererClasses [&NiagaraSpriteRendererProperties]
  renderers [
   {rendererClass=&NiagaraSpriteRendererProperties}
   {rendererClass=&NiagaraSpriteRendererProperties, rendererIndex=1}
  ]
 }
 GoldenDrops {
  emitterUpdateScript [
   &EmitterState()
   &SpawnRate(
    SpawnRate=&FloatFromCurve(FloatCurve=DataInterface() {propertyValues {Curve {keys [<at 0 0>,<at 1 1>]}}}, CurveIndex=$Emitter.Age, Scale Curve=30)
    Use Spawn Probability=false
   )
  ]
  particleSpawnScript [
   &InitializeParticle(
    Color Mode=::NewEnumerator0
    Lifetime=&RandomRangeFloat(Minimum=0.2, Maximum=0.6)
    Sprite Size Mode=::NewEnumerator1
    Uniform Sprite Size Max=5
    Uniform Sprite Size Min=1
    UsePositionOffset=false
    Write Material Random=true
   )
   &PartitionParticles(Write Partition to Ribbon ID=::NewEnumerator0)
   &SocketLocation(
    SamplePosOffset=true
    Sampled Position Offset=&AddVectorToPosition(
     Position=&Multiply_VectorByFloat(Vector=&RandomRangeVector(Minimum=<-1 -1 0>, Maximum=<1 1 1>), Float=7)
     Vector=<0 0 3>
    )
    Sampled Position Offset Coordinate Space:ENiagaraSkelMeshTransforms=::NewEnumerator2
    Socket Index=$Particles.PartitionParticles.Partition
    Socket Reader=$System.Socket Reader
    Socket Sampling Mode=::NewEnumerator3
    VelocitySpeedLimit=false
   )
   &AddVelocity(
    Velocity=&AddVectorToPosition(
     Position=&SelectVectorFromArray(
      Direct Array Index=$Particles.PartitionParticles.Partition
      Vector Selection Array=DataInterface() {propertyValues {FloatData [<0.22 0 1>, <-0.22 0 1>]}}
     )
     Vector=&MakeVector(X=0, Y=&Lerp_Float(A=-0.15, B=-0.4, Alpha=$User.ParaGlideLeanAlpha), Z=0)
    )
    Velocity Speed Scale=&Multiply_Float(
     A=&RandomRangeFloat(Minimum=300, Maximum=700)
     B=&Lerp_Float(A=1, B=1.5, Alpha=$System.Timeline.NormalizedTime)
    )
   )
  ]
  particleUpdateScript [
   &ParticleState()
   &FN_Sprite_Attributes(
    Color=&MultiplyLinearColor(
     Linear Color=&MultiplyLinearColor(
      Linear Color=<rgba 1 1 1 1>
      Scale Factor=&MakeFloat4FromVectorAndFloat("Vector (RGB)"=<1 1 1>, "Float (Alpha)"=&FN_FrontEndFloatModifier(Gameplay=1, Front End=0.5))
     )
     Scale Factor=&MakeVector4_FromLinearColor(LinearColor=&FN_Select_Color_from_Time_Of_Day(
       Use Manual Time of Day Thresholds=true
       Morning=<rgba 1 1 1 1>
       Day=<rgba 1 1 1 1>
       Evening=<rgba 1 1 1 1>
       Night=<rgba 1 1 1 1>
       Evening Threshold=22
      ))
    )
    Create Color Attribute=true
    Create SpriteSize Attribute=true
    SpriteSize=&Multiply_Vector2D(
     A=$Particles.SpriteSize
     B=&MakeVector2D(X=1, Y=&RandomRangeFloat(Minimum=0.5, Maximum=2))
    )
    Write to Module Namespace=true
   )
   &ScaleSpriteSize(Uniform Curve Sprite Scale=DataInterface() {propertyValues {Curve {
       defaultValue=0
       keys [
        {tangentMode=RCTM_User, time=0, value=1}
        {arriveTangent=-0.0405, interpMode=RCIM_Cubic, leaveTangent=-0.0405, tangentMode=RCTM_User, time=0.3, value=1}
        {arriveTangent=-3.0822, interpMode=RCIM_Cubic, leaveTangent=-3.0822, tangentMode=RCTM_User, time=1, value=0}
       ]
      }}})
   &Drag(Drag=&RandomRangeFloat(Minimum=4, Maximum=6), Use Linear Drag=true, Use Rotational Drag=false)
   &CurlNoiseForce(Bias Noise=false, Noise Frequency=10, Noise Strength=500, Pan Noise=false)
   &SolveForcesAndVelocity(Clamp Velocity=false, Limit Acceleration=false)
   &ScaleSpriteSizeBySpeed(Max Scale Factor=<1 5>, Min Scale Factor=<1 2>)
  ]
  rendererClasses [&NiagaraSpriteRendererProperties]
  renderers [{rendererClass=&NiagaraSpriteRendererProperties}]
 }
 Hand_Smoke {
  emitterUpdateScript [
   &EmitterState()
   &SpawnBurst_Instantaneous(Spawn Count=2, Use Spawn Probability=false, UseLoopCountLimit=false)
   &SpawnRate(SpawnRate=3, Use Spawn Probability=false)
  ]
  particleSpawnScript [
   /BRCosmetics/Effects/Niagara/Modules/Groups/FN_SetSpawnGroup_v3/FN_SetSpawnGroup_v3(Number of Groups=2, Pattern:ENiagaraPartitionWaveForm=::NewEnumerator1)
   &FN_Sprite_Attributes(Lifetime:NiagaraFloat=&Lerp_Float(
     A=&RandomRangeFloat(Minimum=1.5, Maximum=2.5)
     B=&RandomRangeFloat(Minimum=0.8, Maximum=1.2)
     Alpha=&FN_FrontEndFloatModifier(Gameplay=$System.Dive, Front End=$User.ParaGlideLeanAlpha)
    )) {module=FN_Sprite_v3}
  ]
  particleUpdateScript [
   &ParticleState()
   &SkeletalMeshLocation(
    Bone / Socket Index=$Particles.SpawnGroup
    Bone Sampling Mode=::NewEnumerator4
    Inherit Velocity=false
    Interpolate Mesh Sampling=false
    MaskSpawnGroup=false
    Sampled Position Offset=&FN_FrontEndVector3Modifier(Gameplay=<0 0 10>, Front End=<0 0 10>)
    Sampled Position Offset Coordinate Space:ENiagaraMeshTransforms=::NewEnumerator2
    Skeletal Mesh=$System.Skeletal Mesh
   )
   &FN_Sprite_Attributes(
    Camera Offset=20
    Create Camera Offset Attribute=true
    Create Sprite Rotation Attribute=true
    Create SpriteSize Attribute=true
    SpriteRotation:NiagaraFloat=&RandomRangeFloat(Minimum=0, Maximum=360)
    SpriteSize=&Vector2DFromFloat(Value=&Multiply_Float(
      A=&OneMinusFloat(Float=&Power(a=&OneMinusFloat(Float=$Particles.NormalizedAge), b=6))
      B=&RandomRangeFloat(Minimum=40, Maximum=65)
     ))
   ) {module=FN_Sprite_v3001}
   &FN_SpriteUV(
    End Frame:NiagaraInt32=&RandomRangeInt(Minimum=55, Maximum=63)
    Randomize Sprite UV scale=true
    Seed=2
    Start Frame:NiagaraInt32=&RandomRangeInt(Minimum=4, Maximum=9)
    SubUV Lookup Index:NiagaraFloat=$Particles.NormalizedAge
    Write to Module Namepsace=true
   )
   /SetVariables(Particles.BlackTurbulentColor:LinearColor=&MakeCustomLinearColorFromBool(
     Bool=true
     True Color=&ColorFromCurve(ColorCurve=DataInterface() {propertyValues {
        AlphaCurve {keys [<at 0 0 RCIM_Cubic>,<at 0.2454 0.9697>,<at 0.722 1>,<at 1 0>]}
        BlueCurve {keys [<at 0 1 RCIM_Cubic>,<at 0.3175 0.03>,<at 0.8819 0.02 RCIM_Cubic>]}
        GreenCurve {keys [<at 0 0.7317 RCIM_Cubic>,<at 0.3175 0.0106>,<at 0.8819 0.02 RCIM_Cubic>]}
        RedCurve {keys [<at 0 0.747 RCIM_Cubic>,<at 0.3175 0.0024>,<at 0.8819 0.02 RCIM_Cubic>]}
       }})
     False Color=&ColorFromCurve(ColorCurve=DataInterface() {propertyValues {
        AlphaCurve {keys [<at 0 0 RCIM_Cubic>,<at 0.2454 0.9697>,<at 0.722 1>,<at 1 0>]}
        BlueCurve {keys [<at 0 1 RCIM_Cubic>,<at 0.3175 0.03>,<at 0.8819 0.02 RCIM_Cubic>]}
        GreenCurve {keys [<at 0 0.7317 RCIM_Cubic>,<at 0.3175 0.0106>,<at 0.8819 0.02 RCIM_Cubic>]}
        RedCurve {keys [<at 0 0.747 RCIM_Cubic>,<at 0.3175 0.0024>,<at 0.8819 0.02 RCIM_Cubic>]}
       }})
    )) {isSetParameters=true}
  ]
  rendererClasses [&NiagaraSpriteRendererProperties]
  renderers [{rendererClass=&NiagaraSpriteRendererProperties}]
 }
 InkTrails {
  emitterUpdateScript [
   &EmitterState(
    Inactive Response=::NewEnumerator1
    Life Cycle Mode=::NewEnumerator1
    Loop Behavior=::NewEnumerator1
    Loop Duration Mode=::NewEnumerator1
    UseLoopDelay=false
   )
   &SpawnBurst_Instantaneous(Spawn Count=2, Use Spawn Probability=false, UseLoopCountLimit=false)
  ]
  particleSpawnScript [
   &InitializeParticle(Lifetime=5, UsePositionOffset=false, Write Material Random=true)
   /SetVariables(Particles.SocketIndex=HlslExpression() {hlslExpression="fmod(Particles.UniqueID,2)"}) {isSetParameters=true}
  ]
  particleUpdateScript [
   /SetVariables(
    Particles.GoopScale:Vector3f=&MakeVector(
     X=1
     Y=1
     Z=&FloatFromCurve(
      FloatCurve=DataInterface() {propertyValues {Curve {keys [
          {
           interpMode=RCIM_Cubic
           leaveTangent=0.1351
           leaveTangentWeight=0.2406
           tangentMode=RCTM_Break
           tangentWeightMode=RCTWM_WeightedBoth
           time=0
           value=0
          }
          {interpMode=RCIM_Cubic, tangentMode=RCTM_Break, time=1, value=1}
         ]}}}
      CurveIndex=$Emitter.Age
      Scale Curve=10
     )
    )
    Particles.Scale=&MakeVector(
     X=1
     Y=1
     Z=&FloatFromCurve(
      FloatCurve=DataInterface() {propertyValues {Curve {keys [
          {
           interpMode=RCIM_Cubic
           leaveTangent=3.3784
           leaveTangentWeight=0.7207
           tangentMode=RCTM_Break
           tangentWeightMode=RCTWM_WeightedBoth
           time=0
           value=0
          }
          {interpMode=RCIM_Cubic, tangentMode=RCTM_Break, time=1, value=1}
         ]}}}
      CurveIndex=$Emitter.Age
      Scale Curve=11
     )
    )
    Particles.Scale2:Vector3f=&MakeVector(
     X=1
     Y=1
     Z=&FloatFromCurve(
      FloatCurve=DataInterface() {propertyValues {Curve {keys [
          {
           interpMode=RCIM_Cubic
           leaveTangent=3.3784
           leaveTangentWeight=0.7207
           tangentMode=RCTM_Break
           tangentWeightMode=RCTWM_WeightedBoth
           time=0
           value=0
          }
          {interpMode=RCIM_Cubic, tangentMode=RCTM_Break, time=1, value=1}
         ]}}}
      CurveIndex=$Emitter.Age
      Scale Curve=8.2
     )
    )
    Particles.Scale3:Vector3f=&MakeVector(
     X=1
     Y=1
     Z=&FloatFromCurve(
      FloatCurve=DataInterface() {propertyValues {Curve {keys [
          {
           interpMode=RCIM_Cubic
           leaveTangent=3.3784
           leaveTangentWeight=0.7207
           tangentMode=RCTM_Break
           tangentWeightMode=RCTWM_WeightedBoth
           time=0
           value=0
          }
          {interpMode=RCIM_Cubic, tangentMode=RCTM_Break, time=1, value=1}
         ]}}}
      CurveIndex=$Emitter.Age
      Scale Curve=3
     )
    )
    Particles.Scale4:Vector3f=&MakeVector(
     X=1
     Y=1
     Z=&FloatFromCurve(
      FloatCurve=DataInterface() {propertyValues {Curve {keys [
          {
           interpMode=RCIM_Cubic
           leaveTangent=3.3784
           leaveTangentWeight=0.7207
           tangentMode=RCTM_Break
           tangentWeightMode=RCTWM_WeightedBoth
           time=0
           value=0
          }
          {interpMode=RCIM_Cubic, tangentMode=RCTM_Break, time=1, value=1}
         ]}}}
      CurveIndex=$Emitter.Age
      Scale Curve=7
     )
    )
   ) {isSetParameters=true}
   /SetVariables(
    Particles.MainTrailColor:LinearColor=&MakeCustomLinearColorFromBool(Bool=$User.bIsFrontEndPreview, True Color=<rgba 0.6152 0.6225 0.7031 1>, False Color=<rgba 1 1 1 1>)
    Particles.ThinTrailColor:LinearColor=&MakeCustomLinearColorFromBool(Bool=$User.bIsFrontEndPreview, True Color=<rgba 0.6152 0.6225 0.7031 1>, False Color=<rgba 1 1 1 1>)
   ) {isSetParameters=true}
   &SampleSkeletalMesh(
    AdditionalYawPitchRoll=false
    Bone / Socket Index:NiagaraInt32=$Particles.SocketIndex
    Bone Sampling Mode:ENiagaraBoneSamplingMode=::NewEnumerator4
    Position Offset Coordinate Space:ENiagaraCoordinateSpace=::Local
    Sampled Position Offset:Vector3f=<0 0 -14.0219>
    Skeletal Mesh:NiagaraDataInterfaceSkeletalMesh=$System.Skeletal Mesh
   )
   &ScaleColor(
    Scale Alpha=&FloatFromCurve(
     FloatCurve=DataInterface() {propertyValues {Curve {keys [
         {
          interpMode=RCIM_Cubic
          leaveTangent=0.1351
          leaveTangentWeight=0.2406
          tangentMode=RCTM_Break
          tangentWeightMode=RCTWM_WeightedBoth
          time=0
          value=0
         }
         {interpMode=RCIM_Cubic, tangentMode=RCTM_Break, time=1, value=1}
        ]}}}
     CurveIndex=$Emitter.Age
    )
    Scale RGB=&MakeVectorFromLinearColorRGB(Color=&FNTimeOfDayColor(Night Color=<rgba 1 1 1 1>, Morning Color=<rgba 1 1 1 1>, Day Color=<rgba 1 1 1 1>, Evening Color=<rgba 1 1 1 1>))
    ScaleA=true
    ScaleRGB=true
   )
   /SetVariables(Particles.Position=&AddVectorToPosition(Position=$StackContext.SampleSkeletalMesh.SampledPosition, Vector=<0 0 12>)) {isSetParameters=true}
   &ParticleState(Kill Particles When Lifetime Has Elapsed=false, Let Infinitely Lived Particles Die When Emitter Deactivates=true)
   &UpdateMeshOrientation(
    Facing Direction=&SelectVectorFromArray(Direct Array Index=$Particles.SocketIndex, Vector Selection Array=DataInterface() {propertyValues {FloatData [<1 0.8 0>, <1 -0.8 0>]}})
    Facing Interpolation=false
    Orientation Method=::NewEnumerator1
    Override Quat=false
    TransformFacing=false
    TransformUp=false
    Up Direction:Vector3f=&MakeVector(X=0, Y=&Lerp_Float(A=-0.2, B=-0.6, Alpha=$User.ParaGlideLeanAlpha), Z=1)
   )
   &DynamicMaterialParameters(
    Index 0 Param 1=&RandomRangeFloat(Minimum=0.5, Maximum=1)
    Index 0 Param 2=&RandomRangeFloat(Minimum=0.8, Maximum=1)
    Index 0 Param 3=&RandomRangeFloat(Minimum=0.5, Maximum=1)
    Index 0 Param 4=&RandomRangeFloat(Minimum=0.8, Maximum=1.2)
    Index 3 Param 1=7
    Index 3 Param 2=17
    Index 3 Param 3=3
    Index 3 Param 4=&RandomRangeFloat(Minimum=0.5, Maximum=1)
    Module.Index0Float0Write=true
    Module.Index0Float1Write=true
    Module.Index0Float2Write=true
    Module.Index0Float3Write=true
    Module.Index3Float0Write=true
    Module.Index3Float1Write=true
    Module.Index3Float2Write=true
    Module.Index3Float3Write=true
    Write Parameter Index 3=::NewEnumerator1
   )
  ]
  rendererClasses [&NiagaraMeshRendererProperties]
  renderers [
   {rendererClass=&NiagaraMeshRendererProperties, rendererIndex=0}
   {rendererClass=&NiagaraMeshRendererProperties, rendererIndex=1}
   {rendererClass=&NiagaraMeshRendererProperties, rendererIndex=2}
   {rendererClass=&NiagaraMeshRendererProperties, rendererIndex=3}
   {rendererClass=&NiagaraMeshRendererProperties, rendererIndex=4}
  ]
 }
 Sign {
  emitterUpdateScript [
   &EmitterState(Life Cycle Mode=::NewEnumerator1, Loop Behavior=::NewEnumerator1, Loop Duration Mode=::NewEnumerator1, UseLoopDelay=false)
   &SpawnBurst_Instantaneous(Spawn Count=2, Use Spawn Probability=false, UseLoopCountLimit=false)
  ]
  particleSpawnScript [
   &InitializeParticle(
    Color=&MakeCustomLinearColorFromBool(Bool=$User.bIsFrontEndPreview, True Color=<rgba 0.5677 0.5677 0.5677 1>, False Color=<rgba 1 1 1 1>)
    Mesh Scale Mode=::NewEnumerator3
    Mesh Uniform Scale=0.115
    Sprite Rotation Mode=::NewEnumerator2
    UsePositionOffset=false
    Write Material Random=true
   )
   &InitialMeshOrientation(
    Mesh Axis To Orient=&SelectVectorFromArray(
     Direct Array Index=&ReturnExecIndex()
     Vector Selection Array=DataInterface() {propertyValues {FloatData [<0 0 -1>, <0 0 1>]}}
    )
    Mesh Orientation Mode=::NewEnumerator2
    Use Rotation Vector=false
   )
  ]
  particleUpdateScript [
   &ParticleState(Kill Particles When Lifetime Has Elapsed=false, Let Infinitely Lived Particles Die When Emitter Deactivates=true)
   &ScaleMeshSize(Scale Factor=&VectorFromFloat(Value=&FloatFromCurve(
      FloatCurve=DataInterface() {propertyValues {Curve {keys [
          {
           interpMode=RCIM_Cubic
           leaveTangent=6.3332
           leaveTangentWeight=0.782
           tangentMode=RCTM_Break
           tangentWeightMode=RCTWM_WeightedBoth
           time=0
           value=0
          }
          {interpMode=RCIM_Cubic, tangentMode=RCTM_Break, time=2, value=1}
         ]}}}
      CurveIndex=$Emitter.Age
     )))
   &UpdateMeshOrientation(Rotation Rate=&Lerp_Float(
     A=&FloatFromCurve(
      FloatCurve=DataInterface() {propertyValues {Curve {keys [<at 0 15 RCIM_Cubic>,<at 1 1 RCIM_Cubic>]}}}
      CurveIndex=$Emitter.Age
      Scale Curve=0.15
     )
     B=0.7
     Alpha=$System.Timeline.NormalizedTime
    ))
   /SetVariables(Particles.Position=&AddVectorToPosition(
     Position=Linked() {linkedVariable {name=System.SkullPosition, type=Vector3f}}
     Vector=&MakeVector(
      X=&Waveform(
       "Waveform [1]"=::NewEnumerator0
       "[1] Phase Scale"=0.4
       "[1] Frequency"=1
       "[1] Amplitude Min/Max"=<-1 1>
       "[1] Scale"=&Lerp_Float(A=8, B=0, Alpha=$User.ParaGlideLeanAlpha)
      )
      Y=0
      Z=&Waveform(
       "Waveform [1]"=::NewEnumerator0
       "[1] Phase Scale"=0.3
       "[1] Frequency"=1
       "[1] Amplitude Min/Max"=<-1 1>
       "[1] Scale"=&Lerp_Float(A=6, B=0, Alpha=$User.ParaGlideLeanAlpha)
      )
     )
    )) {isSetParameters=true, module=SetVariables_#5}
  ]
  rendererClasses [&NiagaraMeshRendererProperties]
  renderers [{rendererClass=&NiagaraMeshRendererProperties, rendererIndex=0}]
 }
 SignDiveAura {
  emitterUpdateScript [
   &EmitterState(Life Cycle Mode=::NewEnumerator1, Loop Behavior=::NewEnumerator1, Loop Duration Mode=::NewEnumerator1, UseLoopDelay=false)
   &SpawnBurst_Instantaneous(Use Spawn Probability=false, UseLoopCountLimit=false)
  ]
  particleSpawnScript [
   &InitializeParticle(
    Mesh Scale=<0.45 0.45 0.2>
    Mesh Scale Mode=::NewEnumerator4
    Sprite Rotation Mode=::NewEnumerator2
    UsePositionOffset=false
    Write Material Random=true
   )
   &InitialMeshOrientation(Mesh Axis To Orient=<0 -1 0>, Mesh Orientation Mode=::NewEnumerator2, Orientation Vector=<0 1 0>, Use Rotation Vector=false)
   /SetVariables(Particles.MeshOrientation2:Quat4f=&MakeQuaternion(
     Quaternion From=0
     Axis=<0 0 1>
     Angle=45
     From=?
     To=?
     Stabilization Mode:ENiagara_OrientVectorCount=::NewEnumerator0
     Side Direction:Vector3f=?
     Up Direction:Vector3f=?
     Angle Type=::NewEnumerator0
     Normalize:NiagaraBool=true
    )) {isSetParameters=true}
  ]
  particleUpdateScript [
   &ParticleState(
    Kill Particles When Lifetime Has Elapsed=false
    Let Infinitely Lived Particles Die When Emitter Deactivates=true
    Loop Particles Lifetime=true
   )
   &ScaleMeshSize(Scale Factor=&VectorFromFloat(Value=&FloatFromCurve(
      FloatCurve=DataInterface() {propertyValues {Curve {keys [
          {
           interpMode=RCIM_Cubic
           leaveTangent=6.3332
           leaveTangentWeight=0.782
           tangentMode=RCTM_Break
           tangentWeightMode=RCTWM_WeightedBoth
           time=0
           value=0
          }
          {interpMode=RCIM_Cubic, tangentMode=RCTM_Break, time=2, value=1}
         ]}}}
      CurveIndex=$Emitter.Age
     ))) {module=ScaleMeshSize001}
   /SetVariables(Particles.Position=&AddVectorToPosition(
     Position=&AddVectorToPosition(
      Position=Linked() {linkedVariable {name=System.SkullPosition, type=Vector3f}}
      Vector=&MakeVector(
       X=&Waveform(
        "Waveform [1]"=::NewEnumerator0
        "[1] Phase Scale"=0.4
        "[1] Frequency"=1
        "[1] Amplitude Min/Max"=<-1 1>
        "[1] Scale"=&Lerp_Float(A=8, B=0, Alpha=$User.ParaGlideLeanAlpha)
       )
       Y=0
       Z=&Waveform(
        "Waveform [1]"=::NewEnumerator0
        "[1] Phase Scale"=0.3
        "[1] Frequency"=1
        "[1] Amplitude Min/Max"=<-1 1>
        "[1] Scale"=&Lerp_Float(A=6, B=0, Alpha=$User.ParaGlideLeanAlpha)
       )
      )
     )
     Vector=<0 0 -2>
    )) {isSetParameters=true, module=SetVariables_#5}
   &ScaleMeshSize(Scale Factor=&MakeVector(X=1, Y=1, Z=&Lerp_Float(A=1, B=2, Alpha=$System.Timeline.NormalizedTime)))
   &DynamicMaterialParameters(
    Index 0 Param 1=&RandomRangeFloat(Minimum=0.5, Maximum=1)
    Index 0 Param 2=&RandomRangeFloat(Minimum=0.8, Maximum=1)
    Index 0 Param 3=&RandomRangeFloat(Minimum=0.5, Maximum=1)
    Index 0 Param 4=&RandomRangeFloat(Minimum=0.5, Maximum=1)
    Index 3 Param 3=5
    Index 3 Param 4=&RandomRangeFloat(Minimum=0.5, Maximum=1)
    Module.Index0Float0Write=true
    Module.Index0Float1Write=true
    Module.Index0Float2Write=true
    Module.Index0Float3Write=true
    Module.Index2Float0Write=true
    Module.Index2Float1Write=true
    Module.Index2Float2Write=true
    Module.Index2Float3Write=true
    Module.Index3Float0Write=true
    Module.Index3Float1Write=true
    Module.Index3Float2Write=true
    Module.Index3Float3Write=true
    Write Parameter Index 2=::NewEnumerator1
    Write Parameter Index 3=::NewEnumerator1
   )
  ]
  rendererClasses [&NiagaraMeshRendererProperties]
  renderers [
   {rendererClass=&NiagaraMeshRendererProperties, rendererIndex=0}
   {rendererClass=&NiagaraMeshRendererProperties, rendererIndex=1}
  ]
 }
 SignGoldenDroplets {
  emitterUpdateScript [
   &EmitterState(Life Cycle Mode=::NewEnumerator1, Loop Duration=1.3315, UseLoopDelay=false)
   &SpawnRate(SpawnRate=5, Use Spawn Probability=false)
  ]
  particleSpawnScript [
   &InitializeParticle(
    Lifetime=&RandomRangeFloat(Minimum=0.5, Maximum=2)
    Mesh Scale Max=<0.45 0.3 0.25>
    Mesh Scale Min=<0.45 0.45 0.2>
    Mesh Scale Mode=::NewEnumerator2
    Sprite Rotation Mode=::NewEnumerator2
    UsePositionOffset=false
    Write Material Random=true
   )
   &InitialMeshOrientation(
    Mesh Orientation Mode=::NewEnumerator2
    Rotation=&RandomRangeVector(Minimum=<0 0 0>, Maximum=<0 0 1>)
    Use Rotation Vector=true
   )
  ]
  particleUpdateScript [
   &ParticleState()
   &UpdateMeshOrientation(Rotation Rate=0.1, Rotation Vector=<0 0 1>)
   &ScaleMeshSize(Scale Factor=&VectorFromFloat(Value=&FloatFromCurve(
      FloatCurve=DataInterface() {propertyValues {Curve {keys [
          {
           interpMode=RCIM_Cubic
           leaveTangent=6.3332
           leaveTangentWeight=0.782
           tangentMode=RCTM_Break
           tangentWeightMode=RCTWM_WeightedBoth
           time=0
           value=0
          }
          {interpMode=RCIM_Cubic, tangentMode=RCTM_Break, time=2, value=1}
         ]}}}
      CurveIndex=$Emitter.Age
     ))) {module=ScaleMeshSize001}
   &ScaleMeshSize(Scale Factor=&MakeVector(
     X=1
     Y=&FloatFromCurve(FloatCurve=DataInterface() {propertyValues {Curve {keys [<at 0.5 1>,<at 1 0.1>]}}})
     Z=&FloatFromCurve(
      FloatCurve=DataInterface() {propertyValues {Curve {keys [<at 0 0 RCIM_Cubic>, {arriveTangent=0.4, interpMode=RCIM_Cubic, leaveTangent=0.4, time=0.5, value=1}, <at 1 0.4 RCIM_Cubic>]}}}
      Scale Curve=&RandomRangeFloat(Minimum=0.6, Maximum=1.6)
     )
    )) {module=ScaleMeshSize002}
   /SetVariables(Particles.Position=&AddVectorToPosition(
     Position=&AddVectorToPosition(
      Position=Linked() {linkedVariable {name=System.SkullPosition, type=Vector3f}}
      Vector=&MakeVector(
       X=&Waveform(
        "Waveform [1]"=::NewEnumerator0
        "[1] Phase Scale"=0.4
        "[1] Frequency"=1
        "[1] Amplitude Min/Max"=<-1 1>
        "[1] Scale"=&Lerp_Float(A=8, B=0, Alpha=$User.ParaGlideLeanAlpha)
       )
       Y=0
       Z=&Waveform(
        "Waveform [1]"=::NewEnumerator0
        "[1] Phase Scale"=0.3
        "[1] Frequency"=1
        "[1] Amplitude Min/Max"=<-1 1>
        "[1] Scale"=&Lerp_Float(A=6, B=0, Alpha=$User.ParaGlideLeanAlpha)
       )
      )
     )
     Vector=<0 0 -2>
    )) {isSetParameters=true, module=SetVariables_#5}
   &DynamicMaterialParameters(
    Index 0 Param 1=&RandomRangeFloat(Minimum=0.5, Maximum=1)
    Index 0 Param 2=&RandomRangeFloat(Minimum=0.8, Maximum=1)
    Index 0 Param 3=&RandomRangeFloat(Minimum=0, Maximum=1)
    Index 0 Param 4=&RandomRangeFloat(Minimum=1, Maximum=2)
    Index 3 Param 3=5
    Index 3 Param 4=&RandomRangeFloat(Minimum=0.5, Maximum=1)
    Module.Index0Float0Write=true
    Module.Index0Float1Write=true
    Module.Index0Float2Write=true
    Module.Index0Float3Write=true
    Module.Index2Float0Write=true
    Module.Index2Float1Write=true
    Module.Index2Float2Write=true
    Module.Index2Float3Write=true
    Module.Index3Float0Write=true
    Module.Index3Float1Write=true
    Module.Index3Float2Write=true
    Module.Index3Float3Write=true
    Write Parameter Index 2=::NewEnumerator1
    Write Parameter Index 3=::NewEnumerator1
   )
   &SolveForcesAndVelocity(Clamp Velocity=false, Limit Acceleration=false)
  ]
  rendererClasses [&NiagaraMeshRendererProperties]
  renderers [
   {rendererClass=&NiagaraMeshRendererProperties, rendererIndex=0}
   {rendererClass=&NiagaraMeshRendererProperties, rendererIndex=1}
  ]
 }
 SignOutline {
  emitterUpdateScript [
   &EmitterState(Life Cycle Mode=::NewEnumerator1, Loop Behavior=::NewEnumerator1, Loop Duration Mode=::NewEnumerator1, UseLoopDelay=false)
   &SpawnBurst_Instantaneous(Use Spawn Probability=false, UseLoopCountLimit=false)
  ]
  particleSpawnScript [
   &InitializeParticle(
    Mesh Scale Mode=::NewEnumerator3
    Mesh Uniform Scale=2.4
    Sprite Rotation Mode=::NewEnumerator2
    UsePositionOffset=false
    Write Material Random=true
   )
   &InitialMeshOrientation(Mesh Axis To Orient=<0 0 1>, Mesh Orientation Mode=::NewEnumerator2, Use Rotation Vector=false)
  ]
  particleUpdateScript [
   &ParticleState(
    Kill Particles When Lifetime Has Elapsed=false
    Let Infinitely Lived Particles Die When Emitter Deactivates=true
    Loop Particles Lifetime=true
   )
   &ScaleMeshSize(Scale Factor=&VectorFromFloat(Value=&FloatFromCurve(
      FloatCurve=DataInterface() {propertyValues {Curve {keys [
          {
           interpMode=RCIM_Cubic
           leaveTangent=6.3332
           leaveTangentWeight=0.782
           tangentMode=RCTM_Break
           tangentWeightMode=RCTWM_WeightedBoth
           time=0
           value=0
          }
          {interpMode=RCIM_Cubic, tangentMode=RCTM_Break, time=2, value=1}
         ]}}}
      CurveIndex=$Emitter.Age
     )))
   /SetVariables(Particles.Position=&AddVectorToPosition(
     Position=Linked() {linkedVariable {name=System.SkullPosition, type=Vector3f}}
     Vector=&MakeVector(
      X=&Waveform(
       "Waveform [1]"=::NewEnumerator0
       "[1] Phase Scale"=0.4
       "[1] Frequency"=1
       "[1] Amplitude Min/Max"=<-1 1>
       "[1] Scale"=&Lerp_Float(A=8, B=0, Alpha=$User.ParaGlideLeanAlpha)
      )
      Y=0
      Z=&Waveform(
       "Waveform [1]"=::NewEnumerator0
       "[1] Phase Scale"=0.3
       "[1] Frequency"=1
       "[1] Amplitude Min/Max"=<-1 1>
       "[1] Scale"=&Lerp_Float(A=6, B=0, Alpha=$User.ParaGlideLeanAlpha)
      )
     )
    )) {isSetParameters=true, module=SetVariables_#5}
   &DynamicMaterialParameters(
    Index 0 Param 1=&RandomRangeFloat(Minimum=0.5, Maximum=1)
    Index 0 Param 2=&RandomRangeFloat(Minimum=0.8, Maximum=1)
    Index 0 Param 3=&RandomRangeFloat(Minimum=0.5, Maximum=1)
    Index 0 Param 4=&RandomRangeFloat(Minimum=0.5, Maximum=1)
    Index 3 Param 1=1
    Index 3 Param 2=1
    Index 3 Param 4=&RandomRangeFloat(Minimum=0.5, Maximum=1)
    Module.Index0Float0Write=true
    Module.Index0Float1Write=true
    Module.Index0Float2Write=true
    Module.Index0Float3Write=true
    Module.Index2Float0Write=true
    Module.Index2Float1Write=true
    Module.Index2Float2Write=true
    Module.Index2Float3Write=true
    Module.Index3Float0Write=true
    Module.Index3Float1Write=true
    Module.Index3Float2Write=true
    Module.Index3Float3Write=true
    Write Parameter Index 2=::NewEnumerator1
    Write Parameter Index 3=::NewEnumerator1
   )
  ]
  rendererClasses [&NiagaraMeshRendererProperties]
  renderers [{rendererClass=&NiagaraMeshRendererProperties, rendererIndex=0}]
 }
}
summary {
 systemName=NS_Contrail_DangerInk
 emitters [
  {emitterName=Decal_Glow, rendererClasses [&NiagaraSpriteRendererProperties]}
  {emitterName=Hand_Smoke, rendererClasses [&NiagaraSpriteRendererProperties]}
  {emitterName=Sign, rendererClasses [&NiagaraMeshRendererProperties]}
  {emitterName=SignOutline, rendererClasses [&NiagaraMeshRendererProperties]}
  {emitterName=GoldenDrops, rendererClasses [&NiagaraSpriteRendererProperties]}
  {emitterName=SignDiveAura, rendererClasses [&NiagaraMeshRendererProperties]}
  {emitterName=InkTrails, rendererClasses [&NiagaraMeshRendererProperties]}
  {emitterName=AppearSmoke, rendererClasses [&NiagaraSpriteRendererProperties]}
  {emitterName=SignGoldenDroplets, rendererClasses [&NiagaraMeshRendererProperties]}
 ]
 userVariables [
  User.bIsFrontEnd {type=NiagaraBool}
  User.bIsFrontEndPreview {type=NiagaraBool}
  User.ParaGlideLeanAlpha {type=NiagaraFloat, defaultValue=0, description=""}
  User.PlayerSpeed {type=NiagaraFloat, defaultValue=0, description=""}
 ]
}
